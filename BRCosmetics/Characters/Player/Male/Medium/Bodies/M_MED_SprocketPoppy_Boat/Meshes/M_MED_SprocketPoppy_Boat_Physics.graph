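# Physics asset for the male medium SprocketPoppy_Boat body cosmetic, simulating two dangling cloth strips (left and right) with rigid-body bone chains.
# Each chain runs four capsules from a small master body constrained tightly to the pelvis, with narrow twist and wider swing on one axis.
# Pelvis and thigh capsules, joined with free rotation, provide body volumes around the cloth.
# Extra free-rotation constraints cross-link the two lower segment pairs, tying left and right chains together.
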PhysicsAsset /BRCosmetics/Characters/Player/Male/Medium/Bodies/M_MED_SprocketPoppy_Boat/Meshes/M_MED_SprocketPoppy_Boat_Physics
bodies {
 dyn_cloth_01_L [{shapeType=Capsule, length=2, radius=3, shapeName=dyn_cloth_01_L_capsule, center=<4.9591 0 0>, rotation=<rot 90 90 -90>}]
 dyn_cloth_01_R [{shapeType=Capsule, length=2, radius=3, shapeName=dyn_cloth_01_R_capsule, center=<4.9591 0 0>, rotation=<rot 90 90 -90>}]
 dyn_cloth_02_L [{shapeType=Capsule, length=2, radius=3.5, shapeName=dyn_cloth_02_L_capsule, center=<4.9591 0 0>, rotation=<rot 90 180 0>}]
 dyn_cloth_02_R [{shapeType=Capsule, length=2, radius=3.5, shapeName=dyn_cloth_02_R_capsule, center=<4.9591 0 0>, rotation=<rot 90 0 180>}]
 dyn_cloth_03_L [{shapeType=Capsule, length=1.5, radius=4, shapeName=dyn_cloth_03_L_capsule, center=<4.9591 0 0>, rotation=<rot 90 -90 90>}]
 dyn_cloth_03_R [{shapeType=Capsule, length=1.5, radius=4, shapeName=dyn_cloth_03_R_capsule, center=<4.9591 0 0>, rotation=<rot 90 -90 90>}]
 dyn_cloth_04_L [{shapeType=Capsule, length=1.5, radius=4, shapeName=dyn_cloth_04_L_capsule, center=<4.9591 0 0>, rotation=<rot 90 -7.5524 172.4476>}]
 dyn_cloth_04_R [{shapeType=Capsule, length=1.5, radius=4, shapeName=dyn_cloth_04_R_capsule, center=<4.9591 0 0>, rotation=<rot 90 0 180>}]
 dyn_cloth_MSTR_L [{shapeType=Capsule, length=0, radius=0.505, shapeName=dyn_cloth_MSTR_L_capsule, rotation=<rot 90 -180 0>}]
 dyn_cloth_MSTR_R [{shapeType=Capsule, length=0, radius=0.505, shapeName=dyn_cloth_MSTR_R_capsule, rotation=<rot 90 -180 0>}]
 pelvis [{shapeType=Capsule, length=15, radius=17, shapeName=pelvis_capsule, center=<-8.7695 -8.1988 0>}]
 thigh_l [{shapeType=Capsule, length=29, radius=9, shapeName=thigh_l_capsule, center=<-20 0 0>, rotation=<rot 0 -90 -90>}]
 thigh_r [{shapeType=Capsule, length=29, radius=9, shapeName=thigh_r_capsule, center=<20 0 0>, rotation=<rot 0 -90 -90>}]
}
constraints [
 {bone1Name=dyn_cloth_01_L, bone2Name=dyn_cloth_MSTR_L, swing1LimitDegrees=10, swing2LimitDegrees=45, twistLimitDegrees=5}
 {bone1Name=dyn_cloth_02_L, bone2Name=dyn_cloth_01_L, swing1LimitDegrees=10, swing2LimitDegrees=45, twistLimitDegrees=5}
 {bone1Name=dyn_cloth_03_L, bone2Name=dyn_cloth_02_L, swing1LimitDegrees=10, swing2LimitDegrees=45, twistLimitDegrees=5}
 {bone1Name=dyn_cloth_04_L, bone2Name=dyn_cloth_03_L, swing1LimitDegrees=10, swing2LimitDegrees=45, twistLimitDegrees=5}
 {bone1Name=dyn_cloth_01_R, bone2Name=dyn_cloth_MSTR_R, swing1LimitDegrees=10, swing2LimitDegrees=45, twistLimitDegrees=5}
 {bone1Name=dyn_cloth_02_R, bone2Name=dyn_cloth_01_R, swing1LimitDegrees=10, swing2LimitDegrees=45, twistLimitDegrees=5}
 {bone1Name=dyn_cloth_03_R, bone2Name=dyn_cloth_02_R, swing1LimitDegrees=10, swing2LimitDegrees=45, twistLimitDegrees=5}
 {bone1Name=dyn_cloth_04_R, bone2Name=dyn_cloth_03_R, swing1LimitDegrees=10, swing2LimitDegrees=45, twistLimitDegrees=5}
 {bone1Name=dyn_cloth_MSTR_L, bone2Name=pelvis, swing1LimitDegrees=5, swing2LimitDegrees=5, twistLimitDegrees=5}
 {bone1Name=dyn_cloth_MSTR_R, bone2Name=pelvis, swing1LimitDegrees=5, swing2LimitDegrees=5, twistLimitDegrees=5}
 {
  bone1Name=thigh_l
  bone2Name=pelvis
  swing1LimitDegrees=45
  swing1Motion=Free
  swing2LimitDegrees=45
  swing2Motion=Free
  twistLimitDegrees=45
  twistMotion=Free
 }
 {
  bone1Name=thigh_r
  bone2Name=pelvis
  swing1LimitDegrees=45
  swing1Motion=Free
  swing2LimitDegrees=45
  swing2Motion=Free
  twistLimitDegrees=45
  twistMotion=Free
 }
 {
  bone1Name=dyn_cloth_03_R
  bone2Name=dyn_cloth_03_L
  swing1LimitDegrees=45
  swing1Motion=Free
  swing2LimitDegrees=45
  swing2Motion=Free
  twistLimitDegrees=75
  twistMotion=Free
 }
 {
  bone1Name=dyn_cloth_04_R
  bone2Name=dyn_cloth_04_L
  swing1LimitDegrees=45
  swing1Motion=Free
  swing2LimitDegrees=45
  swing2Motion=Free
  twistLimitDegrees=10
  twistMotion=Free
 }
]
