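# Animation Blueprint for the SprocketPoppy Boat male body cosmetic, adding cloth physics and pose adjustments on the pose copied from the parent mesh.
# Bone offsets adjust armor, sleeve, belt and cloth, with crouch and DBNO variants.
# Separate RigidBody simulations for normal and DBNO poses blend by DBNO state into a Control Rig.
# Each update computes gravity and RBAN settings through a deprecated helper, plus normalized arm, clavicle and thigh rotation values.
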AnimBlueprint /BRCosmetics/Characters/Player/Male/Medium/Bodies/M_MED_SprocketPoppy_Boat/Meshes/M_MED_SprocketPoppy_Boat_AnimBP
target_skeleton=&/M_MED_SprocketPoppy_Boat_Skeleton
graphs [
 AnimGraph {nodes [
   n1 = Control Rig(Source=Blend Poses by bool(BlendPose_0=Use cached pose 'dbno'(), BlendPose_1=Use cached pose 'common cache'(), bActiveValue=Get bIsDBNO()) {blendType=Linear}) {
    alphaBoolBlend {blendOption=Linear}
    controlRigAssetReference {blueprintRigClass=&/M_MED_SprocketPoppy_Boat_ControlRig.M_MED_SprocketPoppy_Boat_ControlRig_C}
   }
   Output Pose(Result=@n1) {layerGroup=None}
   n2 = "Transform (Modify) Bone - Bone: armor_L"!(ComponentPose=Local To Component(LocalPose=Copy Pose From Mesh() {bUseAttachedParent=true}), Alpha=1, Translation="5,0,0", Rotation="0,0,0") -> (Pose) {
    rotationMode=BMM_Additive
    rotationSpace=BCS_BoneSpace
    translationMode=BMM_Additive
    translationSpace=BCS_BoneSpace
    alphaBoolBlend {blendOption=Linear}
    boneToModify {boneName=armor_L}
    translation {x=5}
   }
   n3 = "Transform (Modify) Bone - Bone: armor_R"!(ComponentPose=@n2, Alpha=1, Translation="5,0,0", Rotation="0,0,0") -> (Pose) {
    rotationMode=BMM_Additive
    rotationSpace=BCS_BoneSpace
    translationMode=BMM_Additive
    translationSpace=BCS_BoneSpace
    alphaBoolBlend {blendOption=Linear}
    boneToModify {boneName=armor_R}
    translation {x=5}
   }
   n4 = "Transform (Modify) Bone - Bone: sleeve_L"!(ComponentPose=@n3, Alpha=1, Translation="5,0,0") -> (Pose) {
    rotationSpace=BCS_WorldSpace
    translationMode=BMM_Additive
    translationSpace=BCS_BoneSpace
    alphaBoolBlend {blendOption=Linear}
    boneToModify {boneName=sleeve_L}
    translation {x=5}
   }
   Save cached pose 'start pose'(Pose=Component To Local(ComponentPose=@n4)) {cacheName=start pose}
   n5 = "Transform (Modify) Bone - Bone: dyn_cloth_MSTR_L"!(ComponentPose=Local To Component(LocalPose=Use cached pose 'start pose'()), Alpha=1, Translation="0,0,0", Rotation="50,0,0") -> (Pose) {
    rotationMode=BMM_Additive
    rotationSpace=BCS_BoneSpace
    translationMode=BMM_Additive
    translationSpace=BCS_BoneSpace
    alphaBoolBlend {blendOption=Linear}
    boneToModify {boneName=dyn_cloth_MSTR_L}
    rotation {pitch=50}
   }
   n6 = "Transform (Modify) Bone - Bone: dyn_cloth_MSTR_R"!(ComponentPose=@n5, Alpha=1, Translation="0,0,0", Rotation="50,0,0") -> (Pose) {
    rotationMode=BMM_Additive
    rotationSpace=BCS_BoneSpace
    translationMode=BMM_Additive
    translationSpace=BCS_BoneSpace
    alphaBoolBlend {blendOption=Linear}
    boneToModify {boneName=dyn_cloth_MSTR_R}
    rotation {pitch=50}
   }
   n7 = "Transform (Modify) Bone - Bone: belt"!(ComponentPose=@n6, Alpha=1, Translation="5,3.5,0", Rotation="0,0,0") -> (Pose) {
    rotationMode=BMM_Additive
    rotationSpace=BCS_BoneSpace
    translationMode=BMM_Additive
    translationSpace=BCS_BoneSpace
    alphaBoolBlend {blendOption=Linear}
    boneToModify {boneName=belt}
    translation=<5 3.5>
   }
   Save cached pose 'crouching'(Pose=Component To Local(ComponentPose=@n7)) {cacheName=crouching}
   n8 = "Transform (Modify) Bone - Bone: dyn_cloth_MSTR_L"(ComponentPose=Local To Component(LocalPose=Use cached pose 'start pose'()), Alpha=1, Translation="0,0,0", Rotation="85,0,0") {
    rotationMode=BMM_Additive
    rotationSpace=BCS_BoneSpace
    translationMode=BMM_Additive
    translationSpace=BCS_BoneSpace
    alphaBoolBlend {blendOption=Linear}
    boneToModify {boneName=dyn_cloth_MSTR_L}
    rotation {pitch=85}
   }
   n9 = "Transform (Modify) Bone - Bone: dyn_cloth_MSTR_R"(ComponentPose=@n8, Alpha=1, Translation="0,0,0", Rotation="85,0,0") {
    rotationMode=BMM_Additive
    rotationSpace=BCS_BoneSpace
    translationMode=BMM_Additive
    translationSpace=BCS_BoneSpace
    alphaBoolBlend {blendOption=Linear}
    boneToModify {boneName=dyn_cloth_MSTR_R}
    rotation {pitch=85}
   }
   n10 = "Transform (Modify) Bone - Bone: belt"(ComponentPose=@n9, Alpha=1, Translation="5,3.5,0", Rotation="0,0,0") {
    rotationMode=BMM_Additive
    rotationSpace=BCS_BoneSpace
    translationMode=BMM_Additive
    translationSpace=BCS_BoneSpace
    alphaBoolBlend {blendOption=Linear}
    boneToModify {boneName=belt}
    translation=<5 3.5>
   }
   n11 = RigidBody!(
    ComponentPose=@n10
    Alpha
    OverrideWorldGravity:Vector="0,0,0"
    SimSpaceSettings:Sim Space Settings Structure=(WorldAlpha=0,VelocityScaleZ=1,DampingAlpha=1,MaxLinearVelocity=10000,MaxAngularVelocity=10000,MaxLinearAcceleration=10000,MaxAngularAcceleration=10000,ExternalLinearDragV=(X=0,Y=0,Z=0),ExternalLinearVelocity=(X=0,Y=0,Z=0),ExternalAngularVelocity=(X=0,Y=0,Z=0))
   ) -> (Pose) {
    bOverrideWorldGravity=true
    overridePhysicsAsset=&/M_MED_SprocketPoppy_Boat_DBNO_Physics
    simulationSpace=BaseBoneSpace
    alphaBoolBlend {blendOption=Linear}
    baseBoneRef {boneName=spine_01}
    componentAppliedLinearAccClamp=<3000 3000 3000>
   }
   Save cached pose 'dbno'(Pose=Component To Local(ComponentPose=@n11)) {cacheName=dbno}
   n12 = Use cached pose 'start pose'()
   n13 = Use cached pose 'crouching'()
   n14 = Get bIsCrouching()
   n15 = RigidBody(
    ComponentPose=Local To Component(LocalPose=Blend Poses by bool(BlendPose_0=@n13, BlendPose_1=@n12, bActiveValue=@n14) {blendType=Linear})
    OverrideWorldGravity="0,0,0"
    SimSpaceSettings=(WorldAlpha=0,VelocityScaleZ=1,DampingAlpha=1,MaxLinearVelocity=10000,MaxAngularVelocity=10000,MaxLinearAcceleration=10000,MaxAngularAcceleration=10000,ExternalLinearDragV=(X=0,Y=0,Z=0),ExternalLinearVelocity=(X=0,Y=0,Z=0),ExternalAngularVelocity=(X=0,Y=0,Z=0))
   ) {
    bOverrideWorldGravity=true
    overridePhysicsAsset=&/M_MED_SprocketPoppy_Boat_Physics
    simulationSpace=BaseBoneSpace
    alphaBoolBlend {blendOption=Linear}
    baseBoneRef {boneName=spine_01}
    componentAppliedLinearAccClamp=<3000 3000 3000>
   }
   Save cached pose 'common cache'(Pose=Component To Local(ComponentPose=@n15)) {cacheName=common cache}
  ]}
 EventGraph {nodes [
   Event BlueprintUpdateAnimation()
   > n1 = Gravity Override State Function DEPRECATED(
    CharacterPartAnimInstance=Self-Reference()
    GameTimeInSeconds=GetGameTimeInSeconds()
    GravityOverrideParams=Get Gravity Override Params()
    RBAN Params=Get RBAN Params()
   )
   > Set Gravity Override(Gravity Override=@n1)
   > Set RBAN Settings Out(RBAN Settings Out:RBANSetting Struct Structure=@n1.RBAN SettingsOut) -> (Output_Get:RBANSetting Struct Structure)
   n2 = GetOwnerSkeletalMeshComponent()
   n3 = GetDeltaTransformFromRefPose!(self=@n2, BoneName=lowerarm_l, BaseName=upperarm_l) -> (
    ReturnValue_Location:Vector
    ReturnValue_Rotation_Roll:float
    ReturnValue_Rotation_Pitch:float
    ReturnValue_Rotation_Yaw:float
    ReturnValue_Scale:Vector
   )
   > Set lowerarm_l_yaw(lowerarm_l_yaw:double=SelectFloat(A=MapRangeClamped(Value=@n3.ReturnValue_Rotation_Yaw, InRangeA=50, InRangeB=100), B=0, bPickA=Get bIsSkydiveFloating())) -> (Output_Get:double)
   n4 = GetDeltaTransformFromRefPose!(self=@n2, BoneName=clavicle_l, BaseName=spine_05) -> (ReturnValue_Location, ReturnValue_Rotation:Rotator, ReturnValue_Scale)
   > Set clavicle_l_ pitch(clavicle_l_ pitch:double=MapRangeClamped(Value=BreakRotator(InRot=@n4.ReturnValue_Rotation).Pitch, InRangeA=0, InRangeB=40)) -> (Output_Get:double)
   n5 = GetDeltaTransformFromRefPose(self=@n2, BoneName=clavicle_r, BaseName=spine_05)
   > Set clavicle_r_pitch(clavicle_r_pitch:double=MapRangeClamped(Value=BreakRotator(InRot=@n5.ReturnValue_Rotation).Pitch, InRangeA=0, InRangeB=40)) -> (Output_Get:double)
   n6 = MapRangeClamped(
    Value=BreakRotator(InRot=GetDeltaTransformFromRefPose(self=@n2, BoneName=thigh_l, BaseName=pelvis).ReturnValue_Rotation).Yaw
    InRangeA=-60
    InRangeB=-90
   )
   n7 = MapRangeClamped(
    Value=BreakRotator(InRot=GetDeltaTransformFromRefPose(self=@n2, BoneName=thigh_r, BaseName=pelvis).ReturnValue_Rotation).Yaw
    InRangeA=-60
    InRangeB=-90
   )
   > Set thigh_yaw(thigh_yaw:double=MapRangeClamped(Value=float + float(A=@n6, B=@n7), InRangeA=0, InRangeB=1)) -> (Output_Get:double)
  ]}
]
variables [
 clavicle_l_ pitch {type=number}
 clavicle_r_pitch {type=number}
 Gravity Override {type=Vector}
 Gravity Override Params {
  type=GravityOverrideParamsStruct
  default {skydive_Parachuting {gravityOverride=<0 -100 -980>, windAmplitude=<100 0 100>, windFrequency=<1 3 3>}}
 }
 lowerarm_l_yaw {type=number}
 RBAN Params {
  type=RBANParamsStruct
  default {
   dBNO {simSpaceSettings {
     maxAngularAcceleration=1
     maxAngularVelocity=3
     maxLinearAcceleration=50
     maxLinearVelocity=700
     velocityScaleZ=1
     externalAngularVelocity=<0 0 0.1>
     externalLinearDragV=<0.2 0.2 0>
    }}
   dBNO_Carried {simSpaceSettings {
     maxAngularAcceleration=1
     maxAngularVelocity=3
     maxLinearAcceleration=50
     maxLinearVelocity=700
     velocityScaleZ=1
     externalAngularVelocity=<0 0 0.1>
     externalLinearDragV=<0.2 0.2 0>
    }}
   emoteOrMelee {simSpaceSettings {
     maxAngularAcceleration=1
     maxAngularVelocity=3
     maxLinearAcceleration=50
     maxLinearVelocity=700
     velocityScaleZ=1
     externalAngularVelocity=<0 0 0.1>
     externalLinearDragV=<0.2 0.2 0>
    }}
   front_End {simSpaceSettings {
     maxAngularAcceleration=1
     maxAngularVelocity=3
     maxLinearAcceleration=50
     maxLinearVelocity=700
     velocityScaleZ=1
     externalAngularVelocity=<0 0 0.1>
     externalLinearDragV=<0.2 0.2 0>
    }}
   onGround_Crouch_Moving {simSpaceSettings {
     maxAngularAcceleration=1
     maxAngularVelocity=3
     maxLinearAcceleration=50
     maxLinearVelocity=700
     velocityScaleZ=1
     externalAngularVelocity=<0 0 0.1>
     externalLinearDragV=<0.2 0.2 0>
    }}
   onGround_Falling {simSpaceSettings {
     maxAngularAcceleration=1
     maxAngularVelocity=3
     maxLinearAcceleration=50
     maxLinearVelocity=700
     velocityScaleZ=1
     externalAngularVelocity=<0 0 0.1>
     externalLinearDragV=<0.2 0.2 0>
    }}
   onGround_Moving {
    componentLinearAccScale=<0 0 0>
    componentLinearVelScale=<0 0 0>
    simSpaceSettings {
     maxAngularAcceleration=1
     maxAngularVelocity=3
     maxLinearAcceleration=50
     maxLinearVelocity=700
     velocityScaleZ=1
     externalAngularVelocity=<0 0 0.1>
     externalLinearDragV=<0.2 0.2 0>
    }
   }
   onGround_Standing {
    componentLinearAccScale=<0 0 0>
    componentLinearVelScale=<0 0 0>
    simSpaceSettings {
     maxAngularAcceleration=1
     maxAngularVelocity=3
     maxLinearAcceleration=50
     maxLinearVelocity=700
     velocityScaleZ=1
     externalAngularVelocity=<0 0 0.1>
     externalLinearDragV=<0.2 0.2 0>
    }
   }
   rail_Grinding {simSpaceSettings {
     maxAngularAcceleration=1
     maxAngularVelocity=3
     maxLinearAcceleration=50
     maxLinearVelocity=700
     velocityScaleZ=1
     worldAlpha=1
     externalAngularVelocity=<0 0 0.1>
     externalLinearDragV=<0.2 0.2 0>
    }}
   riding_Moving {simSpaceSettings {
     maxAngularAcceleration=1
     maxAngularVelocity=3
     maxLinearAcceleration=50
     maxLinearVelocity=700
     velocityScaleZ=1
     externalAngularVelocity=<0 0 0.1>
     externalLinearDragV=<0.2 0.2 0>
    }}
   riding_Moving_Targeting {simSpaceSettings {
     maxAngularAcceleration=1
     maxAngularVelocity=3
     maxLinearAcceleration=50
     maxLinearVelocity=700
     velocityScaleZ=1
     externalAngularVelocity=<0 0 0.1>
     externalLinearDragV=<0.2 0.2 0>
    }}
   riding_Standing {simSpaceSettings {
     maxAngularAcceleration=1
     maxAngularVelocity=3
     maxLinearAcceleration=50
     maxLinearVelocity=700
     velocityScaleZ=1
     externalAngularVelocity=<0 0 0.1>
     externalLinearDragV=<0.2 0.2 0>
    }}
   skydive_DivingDown {
    componentLinearAccScale=<0 0 0>
    componentLinearVelScale=<0 0 0>
    simSpaceSettings {
     maxAngularAcceleration=1
     maxAngularVelocity=3
     maxLinearAcceleration=50
     maxLinearVelocity=700
     velocityScaleZ=1
     externalAngularVelocity=<0 0 0.1>
     externalLinearDragV=<0.2 0.2 0>
    }
   }
   skydive_FloatingDown {
    componentLinearAccScale=<0 0 0>
    componentLinearVelScale=<0 0 0>
    simSpaceSettings {
     maxAngularAcceleration=1
     maxAngularVelocity=3
     maxLinearAcceleration=50
     maxLinearVelocity=700
     velocityScaleZ=1
     externalAngularVelocity=<0 0 0.1>
     externalLinearDragV=<0.2 0.2 0>
    }
   }
   skydive_GoingUp {simSpaceSettings {
     maxAngularAcceleration=1
     maxAngularVelocity=3
     maxLinearAcceleration=50
     maxLinearVelocity=700
     velocityScaleZ=1
     externalAngularVelocity=<0 0 0.1>
     externalLinearDragV=<0.2 0.2 0>
    }}
   skydive_Parachuting {
    componentLinearAccScale=<0 0 0>
    componentLinearVelScale=<0 0 0>
    simSpaceSettings {
     maxAngularAcceleration=1
     maxAngularVelocity=3
     maxLinearAcceleration=50
     maxLinearVelocity=700
     velocityScaleZ=1
     externalAngularVelocity=<0 0 0.1>
     externalLinearDragV=<0.2 0.2 0>
    }
   }
   skydive_Surfing {simSpaceSettings {
     maxAngularAcceleration=1
     maxAngularVelocity=3
     maxLinearAcceleration=50
     maxLinearVelocity=700
     velocityScaleZ=1
     externalAngularVelocity=<0 0 0.1>
     externalLinearDragV=<0.2 0.2 0>
    }}
   swimming {simSpaceSettings {
     maxAngularAcceleration=1
     maxAngularVelocity=3
     maxLinearAcceleration=50
     maxLinearVelocity=700
     velocityScaleZ=1
     externalAngularVelocity=<0 0 0.1>
     externalLinearDragV=<0.2 0.2 0>
    }}
   tactical_Sprint {
    componentLinearAccScale=<0 0 0>
    componentLinearVelScale=<0 0 0>
    simSpaceSettings {
     maxAngularAcceleration=1
     maxAngularVelocity=3
     maxLinearAcceleration=50
     maxLinearVelocity=700
     velocityScaleZ=1
     externalAngularVelocity=<0 0 0.1>
     externalLinearDragV=<0.2 0.2 0>
    }
   }
   ziplining {simSpaceSettings {
     maxAngularAcceleration=1
     maxAngularVelocity=3
     maxLinearAcceleration=50
     maxLinearVelocity=700
     velocityScaleZ=1
     externalAngularVelocity=<0 0 0.1>
     externalLinearDragV=<0.2 0.2 0>
    }}
  }
 }
 RBAN Settings Out {
  type=RBANSettingStruct
  default {
   componentLinearAccScale=<0 0 0>
   componentLinearVelScale=<0 0 0>
   simSpaceSettings {
    dampingAlpha=1
    maxAngularAcceleration=1000
    maxAngularVelocity=3
    maxLinearAcceleration=300
    maxLinearVelocity=700
    velocityScaleZ=0
    worldAlpha=1
    externalAngularVelocity=<0 0 0>
    externalLinearDragV=<0 0.5 0>
    externalLinearVelocity=<0 0 0>
   }
  }
 }
 thigh_yaw {type=number}
]
